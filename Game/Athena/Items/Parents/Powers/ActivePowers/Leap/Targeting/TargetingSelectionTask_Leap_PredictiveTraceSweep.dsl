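; Targeting selection task for Leap active powers that sweeps a capsule trace along the predicted leap path.
; A shared velocity helper uses the ability's launch velocity before launch, then the source actor's velocity, with an upward fallback.
; That velocity sets trace direction, length and an inverse rotation that keeps the capsule world-aligned.
; Source offset lifts grounded pawns' traces off the floor and, before launch, also pushes them along the launch direction.
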(fn GetAbilityVelocity
  (params (Targeting%20Handle :type Targeting%20Request%20Handle%20Structure))
  (bind _targeting_handle (Variables|GetTargeting%20Handle))
  (bind _targeting_handle_2 (Variables|GetTargeting%20Handle))
  (bind _returnvalue (Targeting|GetTargetingSubsystem _targeting_handle_2))
  (bind _returnvalue_2 (TargetingSystem|TargetingTypes|GetTargetingSourceContext _returnvalue _targeting_handle))
  (bind (_sourceactor _instigatoractor _sourcelocation _sourcesocketname _sourceobject) (Utilities|Struct|BreakTargetingSourceContext _returnvalue_2))
  (Utilities|FlowControl|Sequence
    (:then_0
      (Utilities|IsValid _returnvalue
        (:Is%20Valid
          (LocalVariables|SetSourceActor _sourceactor)
          (bind _asga_athena_active_power_leap_parent (Utilities|Casting|CastToGA_Athena_ActivePower_Leap_Parent _sourceobject)
            (:then
              (if (not (Variables|/Game/Athena/Items/Parents/Powers/ActivePowers/Leap/GA_Athena_ActivePower_Leap_Parent.GA_Athena_ActivePower_Leap_Parent_C|GetLaunchStarted _asga_athena_active_power_leap_parent))
                (bind _returnvalue_3 (Class|GAAthenaActivePowerLeapParent|GetLaunchVelocity _asga_athena_active_power_leap_parent))
                (return _returnvalue_3 false)))
            (:CastFailed
              _))
          (Utilities|IsValid (LocalVariables|GetSourceActor)
            (:Is%20Valid
              (return (Transformation|GetVelocity (LocalVariables|GetSourceActor)) true))))))
    (:then_1
      (return "0.000000,0.000000,100.000000" false))))

(fn GetSourceOffset
  (params (TargetingHandle :type Targeting%20Request%20Handle%20Structure%20%28by%20ref%29))
  (bind _targetinghandle (Variables|GetTargetingHandle))
  (bind _returnvalue (Targeting|GetTargetingSubsystem _targetinghandle))
  ;[1 1 1] Add offset during trace before leap so it does't hit what the player is standing on
  (Utilities|IsValid _returnvalue
    (:Is%20Valid
      (bind _output_get (LocalVariables|SetSourceActor (Utilities|Struct|BreakTargetingSourceContext (TargetingSystem|TargetingTypes|GetTargetingSourceContext _returnvalue (Variables|GetTargetingHandle)))))
      (bind _asfort_player_pawn (Utilities|Casting|CastToFortPlayerPawn _output_get)
        (:then
          (LocalVariables|SetScaledHalfHeight (Components|Capsule|GetScaledCapsuleHalfHeight (Variables|/Script/Engine.Character|GetCapsuleComponent _asfort_player_pawn)))
          (if (Math|Boolean|ORBoolean (Utilities|Enum|Equal(Enum) (Variables|/Script/Engine.CharacterMovementComponent|GetMovementMode (Variables|/Script/Engine.Character|GetCharacterMovement _asfort_player_pawn)) "MOVE_Walking") (Utilities|Enum|Equal(Enum) (Variables|/Script/Engine.CharacterMovementComponent|GetMovementMode (Variables|/Script/Engine.Character|GetCharacterMovement _asfort_player_pawn)) "MOVE_NavWalking") (GameplayTags|HasAnyMatchingGameplayTags _asfort_player_pawn (Variables|Default|GetTags_TreatAsOnGround)))
            (LocalVariables|SetMinZOffset (Math|Float|Max(Float) (- (+ (ScalableFloat|GetValueatLevel (Variables|Default|GetRow_MinHeightOffsetFromGround)) (ScalableFloat|GetValueatLevel (Variables|Default|GetRow_TraceHalfHeight))) (LocalVariables|GetScaledHalfHeight))))
            (else
              (LocalVariables|SetMinZOffset 0.0)))
          (bind (_returnvalue_2 _bleapstarted) (CallFunction|GetAbilityVelocity (Variables|GetTargetingHandle)))
          (if _bleapstarted
            (return 0.0 0.0 (LocalVariables|GetMinZOffset))
            (else
              (bind (_returnvalue_x _returnvalue_y _returnvalue_z) (Math|Vector|BreakVector (* (Math|Vector|Normalize _returnvalue_2) (ScalableFloat|GetValueatLevel (Variables|Default|GetRow_SourceOffsetDistance)))))
              (return _returnvalue_x _returnvalue_y (Math|Float|Max(Float) _returnvalue_z (LocalVariables|GetMinZOffset))))))
        (:CastFailed
          _)))))

(fn GetSweptTraceRotation
  (params (TargetingHandle :type Targeting%20Request%20Handle%20Structure%20%28by%20ref%29))
  ;[1 1 1] Force capsule trace to have to world rotation
  (bind _returnvalue (CallFunction|GetAbilityVelocity TargetingHandle))
  (return (Math|Conversions|ToRotator(Quat) (Math|Quat|Inversed(Quat) (Math|Conversions|ToQuaternion(Vector) (Math|Vector|Normalize _returnvalue))))))

(fn GetTraceDirection
  (params (TargetingHandle :type Targeting%20Request%20Handle%20Structure%20%28by%20ref%29))
  (bind _returnvalue (CallFunction|GetAbilityVelocity TargetingHandle))
  (return (Math|Vector|Normalize _returnvalue)))

(fn GetTraceLength
  (params (TargetingHandle :type Targeting%20Request%20Handle%20Structure%20%28by%20ref%29))
  (bind _returnvalue (CallFunction|GetAbilityVelocity TargetingHandle))
  (return (* (Math|Vector|VectorLength _returnvalue) (ScalableFloat|GetValueatLevel (Variables|Default|GetRow_DistanceToCheckAsPctVelocity)))))
